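# Default Lit material for the Quarrel backpack's Niagara mesh particles, showing atlas-based damage and kill reactions with a glitch effect.
# Dynamic parameters supply UVs, glitch values and event times, which drive timed atlas offsets and kill flashing.
# A curve-atlas glitch color feeds base color and emissive blending.
# An alpha-thresholded WPO collapses meshes onto their particle centers, and a normal-versus-local-axis term fakes AO.
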Material /BRCosmetics/Accessories/FORT_Backpacks/Backpack_M_MED_Quarrel/VFX/M_Quarrel_Backpack
graph {
 properties {bUsedWithNiagaraMeshParticles=true, pixelDepthOffsetMode=PDOM_Legacy, usedShadingModels=Default Lit}
 nodes [
  n1 = NamedRerouteUsage()
  n2 = Saturate(None=Add(A=NamedRerouteUsage(), B=ScalarParameter() {parameterName=BaseColorGlitchBlend}))
  LinearInterpolate(A=NamedRerouteUsage(), B=Multiply(A=ComponentMask(None=@n1) {b=true, g=true, r=true}, B=NamedRerouteUsage()), Alpha=@n2)
  n3 = NamedRerouteUsage()
  n4 = ScalarParameter() {defaultValue=0.5, parameterName=GlitchEmissiveStrength}
  MaterialFunctionCall(Emissive=LinearInterpolate(A=Multiply(A=@n3, B=@n4) {constB=0}, Alpha=@n2) {constB=0}) {materialFunction=&MF_FX_EmissiveAdjustments}
  n5 = DynamicParameter() {parameterIndex=1, defaultValue=<rgba 1 10 1 1>, paramNames [glitch_random, kill_time, random, damage_time]}
  NamedRerouteDeclaration(None=@n5.damage_time) {name=damage_time, nodeColor=<rgba 0.1484 1 0 1>}
  NamedRerouteDeclaration(None=@n5.random) {name=particle_random, nodeColor=<rgba 0.2188 0 1 1>}
  n6 = ComponentMask(None=ComponentMask(None=NamedRerouteUsage()) {a=true}) {a=true}
  n7 = Subtract(A=WorldPosition().XYZ, B=ParticlePositionWS())
  NamedRerouteDeclaration(None=Multiply(A=Step(Y=@n6) {constX=0.5}, B=Multiply(A=@n7) {constB=-1})) {name=WPO, nodeColor=<rgba 0.6406 0 1 1>}
  n8 = Constant3Vector() {constant=<rgba 0 1 0 0>}
  n9 = Abs(None=DotProduct(A=Transform(None=@n8) {transformSourceType=TRANSFORMSOURCE_Local}, B=PixelNormalWS()))
  NamedRerouteDeclaration(None=Saturate(None=Add(A=@n9) {constB=0.6})) {name=AO, nodeColor=<rgba 0.6172 0 1 1>}
  n10 = NamedRerouteUsage()
  n11 = NamedRerouteUsage()
  n12 = NamedRerouteUsage()
  n13 = Add(A=Multiply(A=LinearInterpolate(A=@n10, B=@n11, Alpha=@n12)) {constB=0.5}, B=ComponentMask(None=NamedRerouteUsage()) {r=true}) {constA=1}
  n14 = Add(A=Multiply(A=NamedRerouteUsage()) {constB=0.5}, B=ComponentMask(None=NamedRerouteUsage()) {g=true}) {constB=0.875}
  NamedRerouteDeclaration(None=TextureSampleParameter2D(UVs=MaterialFunctionCall(X=@n13, Y=@n14) {materialFunction=&MakeFloat2}.Result) {parameterName=AtlasTexture, texture=&/T_Backpack_M_MED_Quarrel_D}.RGBA) {name=atlas_sample, nodeColor=<rgba 1 0 0.6094 1>}
  n15 = CurveAtlasRowParameter(CurveTime=NamedRerouteUsage()) {
   atlas=&/C_Quarrel_Glitch_Atlas
   curve=&/C_Quarrel_Glitch_01
   parameterName=GlitchColor
  }
  NamedRerouteDeclaration(None=@n15) {name=glitch_color, nodeColor=<rgba 0.0781 0 1 1>}
  NamedRerouteDeclaration(None=@n5.kill_time) {name=kill_time, nodeColor=<rgba 0.8984 1 0 1>}
  NamedRerouteDeclaration(None=@n5.glitch_random) {name=glitch_random, nodeColor=<rgba 0 1 0.4375 1>}
  n16 = DynamicParameter() {defaultValue=<rgba 1 1 4 1>, paramNames [U, V, kill_flash_speed, glitch_amount]}
  n17 = AppendVector(A=@n16.U, B=@n16.U)
  NamedRerouteDeclaration(None=LinearInterpolate(A=Multiply(A=@n17) {constB=0.5}, B=TextureCoordinate(), Alpha=ScalarParameter() {parameterName=UseTexCoords})) {name=UV, nodeColor=<rgba 1 0.2109 0 1>}
  NamedRerouteDeclaration(None=@n16.kill_flash_speed) {name=kill_flash_speed, nodeColor=<rgba 1 0 0.375 1>}
  NamedRerouteDeclaration(None=@n16.glitch_amount) {name=glitch_amount, nodeColor=<rgba 0.2656 1 0 1>}
  n18 = Time()
  n19 = NamedRerouteUsage()
  NamedRerouteDeclaration(None=Step(Y=Subtract(A=@n18, B=@n19), X=ScalarParameter() {defaultValue=0.35, parameterName=damage_time_length}) {constX=0.5, constY=0.5}) {name=damage_index, nodeColor=<rgba 0 1 0.4375 1>}
  n20 = Time()
  n21 = Multiply(A=@n20, B=NamedRerouteUsage())
  n22 = Constant() {r=2}
  NamedRerouteDeclaration(None=Floor(None=Fmod(A=@n21, B=@n22))) {name=kill_index, nodeColor=<rgba 0 0.6016 1 1>}
  n23 = NamedRerouteUsage()
  NamedRerouteDeclaration(None=Step(Y=Subtract(A=@n20, B=@n23), X=ScalarParameter() {defaultValue=3, parameterName=kill_time_length})) {name=index_interp, nodeColor=<rgba 0 0.8594 1 1>}
  NamedRerouteUsage()
  NamedRerouteUsage()
  NamedRerouteUsage()
 ]
}
